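# Level sequence for WildEstate Pre shot A0060_02, a camera move that orbits, then pulls far back and up.
# A wide-lens cine camera and three control meshes (BaseOrbit, PanBoom, TiltRoll) use attach tracks, apparently as a layered rig, alongside a static WorldCube.
# BaseOrbit swings the yaw, PanBoom drives the huge pull-out with slight pan, and TiltRoll adjusts pitch.
# A single camera cut section covers the shot, and a visibility track hides a test scale actor.
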LevelSequence /WildEstate/Sequences/Pre/A0060/WildEstate_Pre_A0060_02
bindings [
 _A0040_WorldCube {
  id=#1
  possessed_class=FortStaticMeshActor
  tracks [
   {type=MovieSceneSpawnTrack, sections [{type=MovieSceneSpawnSection, channels [None {type=MovieSceneScriptingBoolChannel, default=true}]}]}
   {
    type=MovieScene3DTransformTrack
    sections [{
      type=MovieScene3DTransformSection
      channels [
       Location.X {type=MovieSceneScriptingDoubleChannel, default=12209.6585}
       Location.Y {type=MovieSceneScriptingDoubleChannel, default=1384.9789}
       Location.Z {type=MovieSceneScriptingDoubleChannel, default=55.564}
       Rotation.X {type=MovieSceneScriptingDoubleChannel, default=0}
       Rotation.Y {type=MovieSceneScriptingDoubleChannel, default=0}
       Rotation.Z {type=MovieSceneScriptingDoubleChannel, default=0}
       Scale.X {type=MovieSceneScriptingDoubleChannel, default=1}
       Scale.Y {type=MovieSceneScriptingDoubleChannel, default=1}
       Scale.Z {type=MovieSceneScriptingDoubleChannel, default=1}
      ]
     }]
   }
  ]
 }
 Cam_A0060 {
  id=#2
  possessed_class=CineCameraActor
  tracks [
   {
    type=MovieScene3DTransformTrack
    sections [{
      type=MovieScene3DTransformSection
      channels [
       Location.X {type=MovieSceneScriptingDoubleChannel, default=686.5104, const [0,1]}
       Location.Y {type=MovieSceneScriptingDoubleChannel, default=-948.8562, const [0,1]}
       Location.Z {type=MovieSceneScriptingDoubleChannel, default=384.1762, const [0,1]}
       Rotation.X {type=MovieSceneScriptingDoubleChannel, default=0, const [0,1]}
       Rotation.Y {type=MovieSceneScriptingDoubleChannel, default=-15, const [0,1]}
       Rotation.Z {type=MovieSceneScriptingDoubleChannel, default=122.9426, const [0,1]}
       Scale.X {type=MovieSceneScriptingDoubleChannel, default=1, const [1,1]}
       Scale.Y {type=MovieSceneScriptingDoubleChannel, default=1, const [1,1]}
       Scale.Z {type=MovieSceneScriptingDoubleChannel, default=1, const [1,1]}
      ]
     }]
   }
   {type=MovieSceneSpawnTrack, sections [{type=MovieSceneSpawnSection, channels [None {type=MovieSceneScriptingBoolChannel, default=true}]}]}
   {type=MovieScene3DAttachTrack, sections [{type=MovieScene3DAttachSection}]}
  ]
 }
 _1_BaseOrbit_CTRL {
  id=#3
  possessed_class=FortStaticMeshActor
  tracks [
   {type=MovieSceneSpawnTrack, sections [{type=MovieSceneSpawnSection, channels [None {type=MovieSceneScriptingBoolChannel, default=true}]}]}
   {
    type=MovieScene3DTransformTrack
    sections [{
      type=MovieScene3DTransformSection
      channels [
       Location.X {type=MovieSceneScriptingDoubleChannel, default=0, keys [[1001,868.6387],[1253,862.1147],[1646,855.7444],[1941,855.4355]]}
       Location.Y {type=MovieSceneScriptingDoubleChannel, default=0, const [-796.2156,4]}
       Location.Z {type=MovieSceneScriptingDoubleChannel, default=0, const [207.5452,4]}
       Rotation.X {type=MovieSceneScriptingDoubleChannel, default=0, const [0,4]}
       Rotation.Y {type=MovieSceneScriptingDoubleChannel, default=0, const [0,4]}
       Rotation.Z {type=MovieSceneScriptingDoubleChannel, default=0, keys [[1001,-74.5758],[1253,-36.0383],[1646,20.8043],[1941,23.561]]}
       Scale.X {type=MovieSceneScriptingDoubleChannel, default=1, const [1,4]}
       Scale.Y {type=MovieSceneScriptingDoubleChannel, default=1, const [1,4]}
       Scale.Z {type=MovieSceneScriptingDoubleChannel, default=1, const [1,4]}
      ]
     }]
   }
   {type=MovieScene3DAttachTrack, sections [{type=MovieScene3DAttachSection}]}
  ]
 }
 CameraComponent {
  id=#4
  parent_id=#2
  possessed_class=CineCameraComponent
  tracks [
   {
    type=MovieSceneFloatTrack
    display_name=Current Focal Length
    sections [{type=MovieSceneFloatSection, channels [None {type=MovieSceneScriptingFloatChannel, default=14}]}]
   }
   {
    type=MovieSceneFloatTrack
    display_name="Manual Focus Distance (Focus Settings)"
    sections [{type=MovieSceneFloatSection, channels [None {type=MovieSceneScriptingFloatChannel, default=100000}]}]
   }
   {
    type=MovieSceneFloatTrack
    display_name=Current Aperture
    sections [{type=MovieSceneFloatSection, channels [None {type=MovieSceneScriptingFloatChannel, default=2.8}]}]
   }
   {
    type=MovieScene3DTransformTrack
    sections [{
      type=MovieScene3DTransformSection
      channels [
       Location.X {type=MovieSceneScriptingDoubleChannel, default=0}
       Location.Y {type=MovieSceneScriptingDoubleChannel, default=0}
       Location.Z {type=MovieSceneScriptingDoubleChannel, default=0}
       Rotation.X {type=MovieSceneScriptingDoubleChannel, default=0}
       Rotation.Y {type=MovieSceneScriptingDoubleChannel, default=0}
       Rotation.Z {type=MovieSceneScriptingDoubleChannel, default=0}
       Scale.X {type=MovieSceneScriptingDoubleChannel, default=1}
       Scale.Y {type=MovieSceneScriptingDoubleChannel, default=1}
       Scale.Z {type=MovieSceneScriptingDoubleChannel, default=1}
      ]
     }]
   }
  ]
 }
 _3_TiltRoll_CTRL {
  id=#5
  possessed_class=FortStaticMeshActor
  tracks [
   {type=MovieSceneSpawnTrack, sections [{type=MovieSceneSpawnSection, channels [None {type=MovieSceneScriptingBoolChannel, default=true}]}]}
   {
    type=MovieScene3DTransformTrack
    sections [{
      type=MovieScene3DTransformSection
      channels [
       Location.X {type=MovieSceneScriptingDoubleChannel, default=0, const [0,5]}
       Location.Y {type=MovieSceneScriptingDoubleChannel, default=0, const [0,5]}
       Location.Z {type=MovieSceneScriptingDoubleChannel, default=0, const [0,5]}
       Rotation.X {type=MovieSceneScriptingDoubleChannel, default=0, thinned_from=5, keys [[1001,0],[1941,0]]}
       Rotation.Y {
        type=MovieSceneScriptingDoubleChannel
        default=0
        keys [[1001,-25.5264],[1253,-22.5924],[1305,-22.1903],[1377,-29.6167],[1646,-31.2504],[1941,-31.751]]
       }
       Rotation.Z {type=MovieSceneScriptingDoubleChannel, default=0, thinned_from=5, keys [[1001,0],[1941,0]]}
       Scale.X {type=MovieSceneScriptingDoubleChannel, default=1, const [1,5]}
       Scale.Y {type=MovieSceneScriptingDoubleChannel, default=1, const [1,5]}
       Scale.Z {type=MovieSceneScriptingDoubleChannel, default=1, const [1,5]}
      ]
     }]
   }
   {type=MovieScene3DAttachTrack, sections [{type=MovieScene3DAttachSection}]}
  ]
 }
 _2_PanBoom_CTRL {
  id=#6
  possessed_class=FortStaticMeshActor
  tracks [
   {type=MovieSceneSpawnTrack, sections [{type=MovieSceneSpawnSection, channels [None {type=MovieSceneScriptingBoolChannel, default=true}]}]}
   {
    type=MovieScene3DTransformTrack
    sections [{
      type=MovieScene3DTransformSection
      channels [
       Location.X {
        type=MovieSceneScriptingDoubleChannel
        default=0
        thinned_from=5
        keys [[1001,-1473.9471],[1305,-1635.6704],[1646,-65613.681],[1941,-69070.5781]]
       }
       Location.Y {type=MovieSceneScriptingDoubleChannel, default=0, thinned_from=5, keys [[1001,411.5697],[1305,422.4814],[1646,20707.5026],[1941,21803.5547]]}
       Location.Z {
        type=MovieSceneScriptingDoubleChannel
        default=0
        thinned_from=5
        keys [[1001,1424.9623],[1305,1424.9623],[1646,45963.9994],[1941,48370.5586]]
       }
       Rotation.X {type=MovieSceneScriptingDoubleChannel, default=0, const [0,5]}
       Rotation.Y {type=MovieSceneScriptingDoubleChannel, default=0, const [0,5]}
       Rotation.Z {type=MovieSceneScriptingDoubleChannel, default=-6, thinned_from=5, keys [[1001,-3.1947],[1305,-11.7891],[1646,-6.2968],[1941,-6]]}
       Scale.X {type=MovieSceneScriptingDoubleChannel, default=1, const [1,5]}
       Scale.Y {type=MovieSceneScriptingDoubleChannel, default=1, const [1,5]}
       Scale.Z {type=MovieSceneScriptingDoubleChannel, default=1, const [1,5]}
      ]
     }]
   }
   {type=MovieScene3DAttachTrack, sections [{type=MovieScene3DAttachSection}]}
  ]
 }
 TestScaleEditorActor3 {
  id=#7
  possessed_class=TestScaleEditorActor_C
  tracks [{
    type=MovieSceneVisibilityTrack
    sections [{type=MovieSceneVisibilitySection, channels [None {type=MovieSceneScriptingBoolChannel, default=false}]}]
   }]
 }
]
display_rate {denominator=1, numerator=60}
tick_resolution {denominator=1, numerator=24000}
tracks [{type=MovieSceneCameraCutTrack, sections [{type=MovieSceneCameraCutSection, end_frame=2884, start_frame=1001}]}]
